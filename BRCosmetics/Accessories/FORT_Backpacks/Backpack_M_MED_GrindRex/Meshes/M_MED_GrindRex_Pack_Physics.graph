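# Physics asset for the GrindRex backpack cosmetic, giving simulation bodies to the bag, its Hangers and Baseplates parts, and a zipper charm.
# Two boxes form the bag body, with capsule bodies for Hangers and Baseplates each constrained directly to it with narrow swing and wider twist limits.
# The charm is a five-body capsule chain running from charm_basa through zipper_charm to charm_3, free-rotating at the bag and angle-limited between links.
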PhysicsAsset /BRCosmetics/Accessories/FORT_Backpacks/Backpack_M_MED_GrindRex/Meshes/M_MED_GrindRex_Pack_Physics
bodies {
 dyn_Baseplates [{shapeType=Capsule, length=15, radius=3, shapeName=dyn_Baseplates_capsule, center=<0 0 -2.0158>, rotation=<rot 0.0008 0 90>}]
 dyn_Hangers [{shapeType=Capsule, length=10.835, radius=3, shapeName=dyn_Hangers_capsule, center=<0 0 -0.6842>, rotation=<rot 0 0 90>}]
 dyn_bag [
  {
   shapeType=Box
   extentX=26.9729
   extentY=8.8235
   extentZ=30.8446
   shapeName=dyn_bag_box
   center=<5.0951 -1.0041 -7.2352>
   rotation=<rot 0.0003 0 90>
  }
  {shapeType=Box, extentX=21.5, extentY=7.5, extentZ=7.5, shapeName=dyn_bag_box, center=<6.7819 6.1053 -12.4423>}
 ]
 dyn_charm_1 [{shapeType=Capsule, length=0, radius=0.5, shapeName=dyn_charm_1_capsule}]
 dyn_charm_2 [{shapeType=Capsule, length=0, radius=0.5, shapeName=dyn_charm_2_capsule}]
 dyn_charm_3 [{shapeType=Capsule, length=0, radius=3.2, shapeName=dyn_charm_3_capsule, center=<4.3946 0.9624 0>, rotation=<rot 0.0016 0 60>}]
 dyn_charm_basa [{shapeType=Capsule, length=0.505, radius=0.505, shapeName=dyn_charm_basa_capsule}]
 dyn_zipper_charm [{shapeType=Capsule, length=0, radius=0.5, shapeName=dyn_zipper_charm_capsule}]
}
constraints [
 {bone1Name=dyn_Hangers, bone2Name=dyn_bag, swing1LimitDegrees=3, swing2LimitDegrees=0, twistLimitDegrees=20}
 {bone1Name=dyn_Baseplates, bone2Name=dyn_bag, swing1LimitDegrees=3, swing2LimitDegrees=0, twistLimitDegrees=30}
 {bone1Name=dyn_charm_3, bone2Name=dyn_charm_2, swing1LimitDegrees=45, swing2LimitDegrees=45, twistLimitDegrees=6}
 {bone1Name=dyn_zipper_charm, bone2Name=dyn_charm_basa, swing1LimitDegrees=2, swing2LimitDegrees=140, twistLimitDegrees=0}
 {bone1Name=dyn_charm_1, bone2Name=dyn_zipper_charm, swing1LimitDegrees=10, swing2LimitDegrees=40, twistLimitDegrees=10}
 {bone1Name=dyn_charm_2, bone2Name=dyn_charm_1, swing1LimitDegrees=10, swing2LimitDegrees=40, twistLimitDegrees=10}
 {
  bone1Name=dyn_charm_basa
  bone2Name=dyn_bag
  swing1LimitDegrees=45
  swing1Motion=Free
  swing2LimitDegrees=45
  swing2Motion=Free
  twistLimitDegrees=45
  twistMotion=Free
 }
]
